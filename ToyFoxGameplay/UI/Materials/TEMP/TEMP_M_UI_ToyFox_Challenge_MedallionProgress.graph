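# Temporary translucent UI material for a ToyFox challenge medallion progress bar, filled horizontally by a Progress parameter with adjustable edge hardness.
# Static switches select a two- or three-color gradient fill, an optional color backing with flat or gradient alpha, and chamfered end masking.
# An optional animated overlay, masked by the progress edge, adds either noise-modulated scrolling energy rays or scrolling warning chevrons that fade toward both ends.
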Material /ToyFoxGameplay/UI/Materials/TEMP/TEMP_M_UI_ToyFox_Challenge_MedallionProgress
graph {
 properties {blendMode=BLEND_Translucent, materialDomain=MD_UI, pixelDepthOffsetMode=PDOM_Legacy, usedShadingModels=Default Lit}
 nodes [
  n1 = LinearInterpolate(
   A=VectorParameter() {group=Color, parameterName=Progress_Color_Start, defaultValue=<rgba 0.098 0.7451 1 1>}.RGB
   B=VectorParameter() {group=Color, parameterName=Progress_Color_Stop, defaultValue=<rgba 1 0.2346 0.624 1>}.RGB
   Alpha=ComponentMask(None=MaterialFunctionCall() {materialFunction=&GetUserInterfaceUV}.Normalized UV) {r=true}
  )
  NamedRerouteDeclaration(None=StaticSwitchParameter(True=Add(A=@n1, B=@n1), False=@n1) {desc=Double values to have a lighter version, parameterName=AdditiveColors}) {name=2ColorGradient, nodeColor=<rgba 0.4766 1 0 1>}
  n2 = ComponentMask(None=MaterialFunctionCall() {materialFunction=&GetUserInterfaceUV}.Normalized UV) {r=true}
  n3 = ScalarParameter() {defaultValue=1, parameterName=3_Color_Gradient_Multiplier}
  n4 = Power(Base=Saturate(None=Multiply(A=@n2, B=@n3)), Exp=ScalarParameter() {defaultValue=1, parameterName=3_Color_Gradient_ChangeRate})
  NamedRerouteDeclaration(None=MaterialFunctionCall(
    A=VectorParameter() {group=Color, parameterName=Progress_Color_Start, defaultValue=<rgba 0.098 0.7451 1 1>}.RGB
    B=VectorParameter() {group=Color, parameterName=Progress_Color_Middle, defaultValue=<rgba 0.314 0.3278 0.7913 1>}.RGB
    C=VectorParameter() {group=Color, parameterName=Progress_Color_Stop, defaultValue=<rgba 1 0.2346 0.624 1>}.RGB
    Alpha=Saturate(None=@n4)
   ) {materialFunction=&3ColorBlend}.Result) {name=3ColorGradient, nodeColor=<rgba 1 0 0.7969 1>}
  n5 = TextureCoordinate()
  n6 = ComponentMask(None=@n5) {r=true}
  n7 = Constant() {r=32}
  n8 = Add(A=ComponentMask(None=@n5) {g=true}) {constB=-0.33}
  n9 = Subtract(A=Multiply(A=@n6, B=@n7) {constB=20}, B=Multiply(A=@n8))
  n10 = Subtract(A=Multiply(A=OneMinus(None=@n6), B=@n7) {constB=20}, B=Multiply(A=@n8))
  NamedRerouteDeclaration(None=SmoothStep(Value=Multiply(A=@n9, B=@n10)) {constMax=0.05}) {name=ChamferBorder, nodeColor=<rgba 0.7578 1 0 1>}
  n11 = Subtract(A=OneMinus(None=MaterialFunctionCall() {materialFunction=&GetUserInterfaceUV}.Normalized UV))
  n12 = ScalarParameter() {defaultValue=0.7, group=Bar, parameterName=Progress}
  n13 = Multiply(A=ComponentMask(None=Add(A=@n11, B=@n12)) {r=true}, B=ScalarParameter() {defaultValue=128, group=Bar, parameterName=Progress_Edge_Hardness})
  NamedRerouteDeclaration(None=Saturate(None=Multiply(A=@n13))) {name=Progress Fill, nodeColor=<rgba 1 0.5625 0 1>}
  ScalarParameter() {defaultValue=0.5, group=Bar, parameterName=Progress_Backing_Alpha}
  n14 = TextureCoordinate()
  n15 = ComponentMask(None=@n14) {r=true}
  n16 = Time()
  n17 = Frac(None=Multiply(A=@n16) {constB=1.5})
  n18 = Clamp(None=Multiply(B=Cosine(None=@n17)) {constA=1}) {maxDefault=0.6, minDefault=0.15}
  n19 = Multiply(A=@n16) {constB=-0.2}
  n20 = TextureSample(UVs=MaterialFunctionCall(R=Add(A=@n19, B=@n15), G=ComponentMask(None=@n14) {g=true}) {materialFunction=&AppendMany}.RG) {texture=/ToyFoxGameplay/UI/Textures/TEMP/TEMP_T_UI_ToyFox_MedallionBar}
  n21 = TextureCoordinate()
  n22 = AppendVector(A=Multiply(A=ComponentMask(None=@n21) {r=true}) {constB=1.1}, B=Multiply(A=ComponentMask(None=@n21) {g=true}) {constB=0.18})
  n23 = Multiply(
   A=OneMinus(None=SmoothStep(Min=@n18, Value=@n20.A) {constMax=0.1, constMin=0.8})
   B=TextureSample(UVs=Panner(Coordinate=@n22) {speedX=0.6}) {samplerType=SAMPLERTYPE_Grayscale, texture=/ToyFoxGameplay/UI/Textures/TEMP/TEMP_T_UI_ToyFox_MedallionBar_Noise}.R
  )
  n24 = OneMinus(None=NamedRerouteUsage())
  n25 = Multiply(
   A=LinearInterpolate(
    A=VectorParameter() {group=Color, parameterName=EnergyRays_Color1, defaultValue=<rgba 0.0545 0.5711 0.7913 1>}.RGB
    B=VectorParameter() {group=Color, parameterName=EnergyRays_Color2, defaultValue=<rgba 0.8796 0.1248 0.4852 1>}.RGB
    Alpha=@n15
   )
   B=Multiply(A=LinearInterpolate(A=@n23, Alpha=@n24) {constAlpha=0, constB=0}) {constB=2.5, desc=Rays opacity}
  )
  n26 = VectorParameter() {group=Color, parameterName=Backing_Color, defaultValue=<rgba 0 0 0 1>}
  n27 = VectorParameter() {group=Color, parameterName=Backing_Color_Stop, defaultValue=<rgba 0 0 0 1>}
  n28 = ScalarParameter() {parameterName=AnimateBackingColor}
  n29 = StaticSwitchParameter(True=NamedRerouteUsage(), False=NamedRerouteUsage()) {parameterName=3_Color_Gradient}
  n30 = StaticSwitchParameter(
   True=LinearInterpolate(A=LinearInterpolate(A=@n26.RGB, B=@n27.RGB, Alpha=@n28), B=@n29, Alpha=NamedRerouteUsage())
   False=Multiply(A=@n29, B=NamedRerouteUsage())
  ) {parameterName=HasBacking}
  n31 = Time()
  n32 = TextureCoordinate() {desc=texture itself, uTiling=20, vTiling=0.5}
  n33 = MaterialFunctionCall(R=Add(A=Multiply(A=@n31) {constB=0.2}, B=ComponentMask(None=@n32) {r=true}), G=Add(A=ComponentMask(None=@n32) {g=true}) {constB=-0.25}) {materialFunction=&AppendMany}
  n34 = ComponentMask(None=TextureCoordinate() {desc=linear gradient alpha for the texture}) {r=true}
  n35 = Multiply(A=Min(A=@n34, B=OneMinus(None=@n34))) {constB=3}
  n36 = LinearInterpolate(
   A=Multiply(
    A=TextureSample(UVs=@n33.RG) {samplerType=SAMPLERTYPE_Grayscale, texture=/ToyFoxGameplay/UI/Textures/TEMP/TEMP_T_UI_ToyFox_MedallionBar_Warning_Chevron}.R
    B=Power(Base=@n35)
   )
   Alpha=OneMinus(None=NamedRerouteUsage())
  ) {constAlpha=0, constB=0}
  n37 = Multiply(
   A=VectorParameter() {group=Color, parameterName=ChevronColor, defaultValue=<rgba 1 0 0.0252 1>}.RGB
   B=Multiply(A=Saturate(None=@n36)) {constB=0.3, desc=Chevron opacity}
  )
  StaticSwitchParameter(
   True=StaticSwitchParameter(True=Add(A=@n25, B=@n30), False=Add(A=@n37, B=@n30)) {desc=True - Energy Bars. False - Chevron, parameterName=EnergyRay}
   False=@n30
  ) {desc=Should use animated fill texture or not, parameterName=FillTexture}
  n38 = OneMinus(None=ComponentMask(None=TextureCoordinate() {uTiling=0.5}) {r=true})
  n39 = LinearInterpolate(
   A=StaticSwitchParameter(True=Add(A=@n38) {constB=-0.5}, False=ScalarParameter() {defaultValue=0.5, group=Bar, parameterName=Progress_Backing_Alpha}) {parameterName=Gradient Backing}
   B=ScalarParameter() {defaultValue=1, group=Bar, parameterName=Progress_Fill_Alpha}
   Alpha=NamedRerouteUsage()
  ) {constAlpha=1}
  StaticSwitchParameter(True=Multiply(A=NamedRerouteUsage(), B=@n39), False=@n39) {parameterName=Chamfer Borders}
 ]
}
