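# Temporary translucent UI material for a ToyFox challenge medallion pulse, built from animated diamond shapes.
# Two MF_DiamondPulse calls share start scale, rotation and timing parameters, with the second having its own end scale, much thinner end width and falloff.
# Max merges the pulses and a constant multiply dims them.
# The graph shows no connection from that result, PulseColor or the extra TimeOffset parameter to material outputs.
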Material /ToyFoxGameplay/UI/Materials/TEMP/TEMP_M_UI_ToyFox_Challenge_MedallionPulse
graph {
 properties {blendMode=BLEND_Translucent, materialDomain=MD_UI, pixelDepthOffsetMode=PDOM_Legacy, usedShadingModels=Default Lit}
 nodes [
  n1 = ScalarParameter() {group=PulseStyle, parameterName="ScaleStart (S)"}
  n2 = ScalarParameter() {defaultValue=0.75, group=PulseStyle, parameterName=ScaleEnd}
  n3 = ScalarParameter() {defaultValue=0.3, group=PulseStyle, parameterName="WidthEnd (S)"}
  n4 = Constant() {r=0.125}
  n5 = ScalarParameter() {group=PulseAnim, parameterName="TimeOffset (S)"}
  n6 = ScalarParameter() {group=PulseAnim, parameterName="ManualTime (S)"}
  n7 = ScalarParameter() {defaultValue=0.6, group=PulseStyle, parameterName=ScaleEnd2}
  n8 = ScalarParameter() {defaultValue=0.01, group=PulseStyle, parameterName=WidthEnd2}
  n9 = ScalarParameter() {defaultValue=0.2605, group=PulseStyle, parameterName="PulseAnimFalloff (S)"}
  Multiply(A=Max(
    A=MaterialFunctionCall!(
     PulseSpeed
     ScaleStart=@n1
     ScaleEnd=@n2
     WidthStart
     WidthEnd=@n3
     PulseAnimFalloff
     Rotation=@n4
     TimeOffset=@n5
     ManualTime=@n6
     OpacityStart
     OpacityEnd
     TimeSpeed
     TexCoord
    ) {materialFunction=/Game/Athena/HUD/LTM/Infiltration/Materials/MF_DiamondPulse}.Result
    B=MaterialFunctionCall(ScaleStart=@n1, ScaleEnd=@n7, WidthEnd=@n8, PulseAnimFalloff=@n9, Rotation=@n4, TimeOffset=@n5, ManualTime=@n6) {materialFunction=&MF_DiamondPulse}.Result
   )) {constB=0.7}
  ScalarParameter() {defaultValue=0.1, group=PulseStyle, parameterName=TimeOffset}
  VectorParameter() {parameterName=PulseColor, defaultValue=<rgba 0.733 0.5622 0.0498 1>}
 ]
}
